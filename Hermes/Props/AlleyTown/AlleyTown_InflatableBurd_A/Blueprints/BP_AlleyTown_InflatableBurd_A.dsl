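; BP_AlleyTown_InflatableBurd_A is an AlleyTown prop Blueprint for an inflatable Burd.
; Its construction script calls the parent, then sets material slot 1 on the static mesh component to the B material instance.
; BeginPlay, overlap and Tick events are empty, so the Blueprint adds no runtime logic of its own.
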(event EventBeginPlay)

(event Collision|EventActorBeginOverlap
  (params (OtherActor :type Actor%20Object%20Reference)))

(event EventTick
  (params (DeltaSeconds :type Float%20%28single-precision%29)))

(fn ConstructionScript
  (|CallParentFunction)
  (Rendering|Material|SetMaterial (Variables|Debug|GetStaticMeshComponent) 1 "/Hermes/Props/AlleyTown/AlleyTown_InflatableBurd_A/Materials/MI_AlleyTown_InflatableBurd_A_B.MI_AlleyTown_InflatableBurd_A_B"))
